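; Gameplay effect custom application requirement that decides whether the Figment low gravity effect can apply, as part of the WackyWeekend playlist modifiers.
; It reads the WackyWeekend.LowGrav.Enabled row from the FigmentGameData curve table.
; Application passes only when that value equals 1.0, and fails when it differs or the row is missing.
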(fn CanApplyGameplayEffect
  (params (GameplayEffect :type Gameplay%20Effect%20Object%20Reference) (Spec :type Gameplay%20Effect%20Spec%20Structure%20%28by%20ref%29) (ASC :type Ability%20System%20Component%20Object%20Reference))
  (bind _outxy (DataTable|EvaluateCurveTableRow "/Figment_LootTables/DataTables/FigmentGameData.FigmentGameData" "WackyWeekend.LowGrav.Enabled")
    (:RowFound
      (return (== _outxy 1.0)))
    (:RowNotFound
      (return false))))
